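# Sound cue for a boulder being dislodged, layering close and distant versions of the sound.
# Each layer picks randomly from three waves, then passes through its own large-tree-destruction attenuation and a ranged-weapon-fire sound class override.
# A mixer blends both, with the distant input lower despite its boosted modulator, under slight overall pitch variation.
# An object-impact concurrency set limits simultaneous plays, and the debug flag is left on.
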SoundCue /Game/Sounds/Fort_Impact_Sounds/Demolition/Boulder/Boulder_Dislodge_Cue
root=Modulator_7
nodes [
 {
  id=Attenuation_2
  children [Modulator_5]
  properties {attenuationSettings=/Game/Sounds/Fort_Impact_Sounds/Demolition/Tree/Demo_Tree_Large_Destroyed_Close_Att}
 }
 {
  id=Attenuation_7
  children [Modulator_8]
  properties {attenuationSettings=/Game/Sounds/Fort_Impact_Sounds/Demolition/Tree/Demo_Tree_Large_Destroyed_Distant_Att}
 }
 {id=Mixer_1, children [SoundClass_0,SoundClass_2], properties {inputVolume [1,0.4]}}
 {id=Modulator_5, children [Random_0], properties {pitchMax=1.1, pitchMin=1}}
 {id=Modulator_7, children [Mixer_1], properties {pitchMax=1.1, pitchMin=1}}
 {id=Modulator_8, children [Random_1], properties {volumeMax=2, volumeMin=2}}
 {id=Random_0, children [WavePlayer_0,WavePlayer_1,WavePlayer_2], properties {weights [1,1,1]}}
 {id=Random_1, children [WavePlayer_3,WavePlayer_4,WavePlayer_5], properties {weights [1,1,1]}}
 {
  id=SoundClass_0
  children [Attenuation_2]
  properties {soundClassOverride=&Fort_WeaponRanged_Fire_3P_PSM}
 }
 {
  id=SoundClass_2
  children [Attenuation_7]
  properties {soundClassOverride=&Fort_WeaponRanged_Fire_Distant}
 }
 {id=WavePlayer_0, properties {soundWaveAssetPtr=&/Boulder_Dislodge_02}}
 {id=WavePlayer_1, properties {soundWaveAssetPtr=&/Boulder_Dislodge_01}}
 {id=WavePlayer_2, properties {soundWaveAssetPtr=&/Boulder_Dislodge_03}}
 {id=WavePlayer_3, properties {soundWaveAssetPtr=&/Boulder_Dislodge_Distant01}}
 {id=WavePlayer_4, properties {soundWaveAssetPtr=&/Boulder_Dislodge_Distant02}}
 {id=WavePlayer_5, properties {soundWaveAssetPtr=&/Boulder_Dislodge_Distant03}}
]
properties {
 bDebug=true
 maxDistance=18000
 volumeMultiplier=0.7
 concurrencyOverrides {resolutionRule=PreventNew}
 concurrencySet [&CONC_Object_Impacts_FourMax]
}
