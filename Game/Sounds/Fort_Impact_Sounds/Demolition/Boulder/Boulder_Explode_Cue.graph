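# Sound cue for an exploding boulder, layering randomized close and distant explosion sounds.
# A mixer combines one of three close waves with one of three distant waves at half volume, the latter slightly delayed, under modulators for variation.
# Each layer borrows large-tree destruction attenuation and a weapon-fire sound class override.
# The cue references a four-max object-impact concurrency set and has its debug flag on.
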SoundCue /Game/Sounds/Fort_Impact_Sounds/Demolition/Boulder/Boulder_Explode_Cue
root=Modulator_7
nodes [
 {
  id=Attenuation_2
  children [Modulator_5]
  properties {attenuationSettings=/Game/Sounds/Fort_Impact_Sounds/Demolition/Tree/Demo_Tree_Large_Destroyed_Close_Att}
 }
 {
  id=Attenuation_7
  children [Modulator_8]
  properties {attenuationSettings=/Game/Sounds/Fort_Impact_Sounds/Demolition/Tree/Demo_Tree_Large_Destroyed_Distant_Att}
 }
 {id=Delay_0, children [SoundClass_2], properties {delayMax=0.1, delayMin=0.05}}
 {id=Mixer_1, children [SoundClass_0,Delay_0], properties {inputVolume [1,0.5]}}
 {id=Modulator_5, children [Random_2], properties {pitchMax=1.1, pitchMin=0.9}}
 {id=Modulator_7, children [Mixer_1]}
 {id=Modulator_8, children [Random_3], properties {volumeMax=1, volumeMin=1}}
 {id=Random_2, children [WavePlayer_6,WavePlayer_7,WavePlayer_8], properties {weights [1,1,1]}}
 {id=Random_3, children [WavePlayer_9, WavePlayer_10, WavePlayer_11], properties {weights [1,1,1]}}
 {
  id=SoundClass_0
  children [Attenuation_2]
  properties {soundClassOverride=&Fort_WeaponRanged_Fire_3P_PSM}
 }
 {
  id=SoundClass_2
  children [Attenuation_7]
  properties {soundClassOverride=&Fort_WeaponRanged_Fire_Distant}
 }
 {id=WavePlayer_10, properties {soundWaveAssetPtr=&/Boulder_Explode_Distant_02}}
 {id=WavePlayer_11, properties {soundWaveAssetPtr=&/Boulder_Explode_Distant_03}}
 {id=WavePlayer_6, properties {soundWaveAssetPtr=&/Boulder_Explode_01}}
 {id=WavePlayer_7, properties {soundWaveAssetPtr=&/Boulder_Explode_02}}
 {id=WavePlayer_8, properties {soundWaveAssetPtr=&/Boulder_Explode_03}}
 {id=WavePlayer_9, properties {soundWaveAssetPtr=&/Boulder_Explode_Distant_01}}
]
properties {
 bDebug=true
 maxDistance=18000
 volumeMultiplier=0.4
 concurrencyOverrides {resolutionRule=PreventNew}
 concurrencySet [&CONC_Object_Impacts_FourMax]
}
