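# Sound cue for a boulder being hit, in the demolition impact sounds.
# A mixer layers a random heavy medium-stone physics impact, boosted, over a quieter random boulder hit sample.
# A modulator varies pitch slightly on each play, on top of raised cue pitch and volume.
# It plays in the impacts sound class with small-object demolition attenuation and a concurrency set that limits simultaneous object impacts.
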SoundCue /Game/Sounds/Fort_Impact_Sounds/Demolition/Boulder/Boulder_Hit_Cue
root=Modulator_2
nodes [
 {id=Mixer_1, children [Random_5,Random_6], properties {inputVolume [1.3,0.5]}}
 {id=Modulator_2, children [Mixer_1], properties {pitchMax=1.07, pitchMin=0.92}}
 {id=Random_5, children [WavePlayer_21, WavePlayer_22, WavePlayer_23, WavePlayer_24], properties {weights [1,1,1,1]}}
 {id=Random_6, children [WavePlayer_25, WavePlayer_26, WavePlayer_27], properties {weights [1,1,1]}}
 {id=WavePlayer_21, properties {soundWaveAssetPtr=/Game/Sounds/Physics/Objects/Stone/StoneMedium/Impacts/Phys_Stone_Medium_Impact_Heavy_04}}
 {id=WavePlayer_22, properties {soundWaveAssetPtr=/Game/Sounds/Physics/Objects/Stone/StoneMedium/Impacts/Phys_Stone_Medium_Impact_Heavy_01}}
 {id=WavePlayer_23, properties {soundWaveAssetPtr=/Game/Sounds/Physics/Objects/Stone/StoneMedium/Impacts/Phys_Stone_Medium_Impact_Heavy_02}}
 {id=WavePlayer_24, properties {soundWaveAssetPtr=/Game/Sounds/Physics/Objects/Stone/StoneMedium/Impacts/Phys_Stone_Medium_Impact_Heavy_03}}
 {id=WavePlayer_25, properties {soundWaveAssetPtr=&/Boulder_Hit_01}}
 {id=WavePlayer_26, properties {soundWaveAssetPtr=&/Boulder_Hit_02}}
 {id=WavePlayer_27, properties {soundWaveAssetPtr=&/Boulder_Hit_03}}
]
properties {
 attenuationSettings=/Game/Sounds/Attenuation/Fort_Demolition_Hit_Small_Sized_Object_Att
 maxDistance=4000
 pitchMultiplier=1.1
 soundClassObject=&Fort_Impacts
 volumeMultiplier=1.44
 concurrencyOverrides {resolutionRule=PreventNew}
 concurrencySet [&CONC_Object_Impacts_FourMax]
}
